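# Niagara skydiving contrail for the BrokenGravel cosmetic, with a red drone, a HUD, body glow and scan, and GPU speed lines attached to the character's skeleton.
# FN_Contrail_Setup's dive value drives speed-line spawning.
# ParaGlideLeanAlpha blends drone motion, body-effect intensity and HUD pulsing.
# Front-end modifiers and a preview flag adjust the look outside gameplay, and time of day scales some HUD brightness.
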NiagaraSystem /BRCosmetics/Athena/Effects/Particles/Skydiving/NS_Contrail_BrokenGravel/NS_Contrail_BrokenGravel
dependencies {
 usedDataInterfaces [
  &NiagaraDataInterfaceCurve
  &NiagaraDataInterfaceArrayFloat3
  &NiagaraDataInterfaceArrayInt32
  &NiagaraDataInterfaceVector2DCurve
  &NiagaraDataInterfaceSkeletalMesh
 ]
 usedDynamicInputs [&Vector2DFromCurve]
 usedModules [
  &self:SystemSpawnScript.SystemScriptSource.SystemScriptGraph.NiagaraNodeAssignment_0.SetVariables_#1
  &SystemState
  /Game/Effects/Niagara/Modules/Contrails/FN_Contrail_Setup
  &FN_Integrator
 ]
}
emitters {
 FN_SpeedLines {
  simTarget=GPUComputeSim
  emitterUpdateScript [
   &EmitterState(Life Cycle Mode=::NewEnumerator1, Loop Duration=1, UseLoopDelay=false)
   &SpawnRate(
    Spawn Probability=&MakeCustomFloatFromBool(Bool=&SetBoolByFloatComparison(A=$System.Age, B=0.2), True Float=1, False Float=1)
    SpawnRate=&Lerp_Float(
     A=0
     B=&FN_FrontEndFloatModifier(
      Gameplay=&Multiply_Float(A=55, B=$System.FN_Contrail_Setup.Dive Value)
      Front End=&Multiply_Float(A=50, B=&Lerp_Float(A=0.5, B=1, Alpha=$System.FN_Contrail_Setup.Dive Value))
     )
     Alpha=1
    )
    Use Spawn Probability=true
   )
  ]
  particleSpawnScript [
   &InitializeParticle(
    Color Maximum=<rgba 0.25 0.25 0.25 0.05>
    Color Minimum=<rgba 1 1 1 0.025>
    Color Mode=::NewEnumerator2
    Lifetime Max=0.5
    Lifetime Min=0.3
    Lifetime Mode=::NewEnumerator1
    Sprite Rotation Mode=::NewEnumerator2
    Sprite Size=&UniformRangedVector2D(Minimum=<7 100>, Maximum=<1 200>)
    Sprite Size Mode=::NewEnumerator4
    UsePositionOffset=false
    Write Material Random=true
   ) {module=InitializeParticle001}
   &SkeletalMeshLocation(
    Bone / Socket Index=3
    Bone Sampling Mode=::NewEnumerator4
    MaskSpawnGroup=false
    Sampled Orientation Coordinate Space=::Local
    Skeletal Mesh=$System.SkeletalMesh
   )
   &SphereLocation(
    MaskSpawnGroup=false
    Non Uniform Scale=&RotateVectorByQuaternion(Vector To Rotate=<1 1 1>, Quaternion=$Particles.SkeletalMeshLocation.SampledRotation)
    Offset=&FN_FrontEndVector3Modifier(
     Gameplay=&Add_Vector(
      A=&RotateVectorByQuaternion(Vector To Rotate=<100 0 0>, Quaternion=$Particles.SkeletalMeshLocation.SampledRotation)
      B=&TransformVector(Vector=<0 0 -200>, Source Space=::World, Destination Space=::World)
     )
     Front End=<0 0 0>
    )
    Override Seed=false
    Sphere Orientation Axis:Vector3f=&RotateVectorByQuaternion(Vector To Rotate=<100 0 0>, Quaternion=$Particles.SkeletalMeshLocation.SampledRotation)
    Sphere Origin:NiagaraPosition=$Particles.SkeletalMeshLocation.SampledPosition
    Sphere Radius=250
    Surface Only=false
    UseNonUniformScale=true
    UseOffset=true
    UseSphereOrientation=true
   )
   &InheritVelocity(Module.Limit Speed=true)
   &AddVelocity(
    Velocity=&FN_FrontEndVector3Modifier(Gameplay=<0 0 0>, Front End=<0 -100 100>)
    Velocity Speed Scale=&FN_FrontEndFloatModifier(Gameplay=1, Front End=2)
   )
  ]
  particleUpdateScript [
   &ParticleState()
   &SolveForcesAndVelocity(Clamp Velocity=false, Limit Acceleration=false)
   &ScaleSpriteSize(
    Non-Uniform Curve Scale=&FN_FrontEndVector2Modifier(Front End:Vector2f=<0.5 2>, Gameplay:Vector2f=<1.5 3>)
    Non-Uniform Curve Sprite Scale=DataInterface() {propertyValues {
      XCurve {keys [<at -0.0069 0.5 RCIM_Cubic>,<at 0.4918 1 RCIM_Cubic>,<at 1 0.5 RCIM_Cubic>]}
      YCurve {keys [
        {
         arriveTangent=3.1787
         arriveTangentWeight=0.684
         interpMode=RCIM_Cubic
         leaveTangent=3.1787
         leaveTangentWeight=0.684
         tangentMode=RCTM_User
         tangentWeightMode=RCTWM_WeightedBoth
         time=0
         value=0.25
        }
        {
         arriveTangentWeight=0.2354
         interpMode=RCIM_Cubic
         leaveTangentWeight=0.3527
         tangentMode=RCTM_Break
         tangentWeightMode=RCTWM_WeightedBoth
         time=0.4975
         value=1
        }
        <at 1 0.25 RCIM_Cubic>
       ]}
     }}
    Scale Sprite Size Mode=::NewEnumerator3
   )
   &ScaleColor(
    Scale Alpha=&FloatFromCurve(
     FloatCurve=DataInterface() {propertyValues {Curve {
        defaultValue=1
        keys [
         {arriveTangent=9.0725, interpMode=RCIM_Cubic, leaveTangent=9.0725, tangentMode=RCTM_User, time=0, value=0}
         {arriveTangent=-0.0194, interpMode=RCIM_Cubic, leaveTangent=-0.0194, tangentMode=RCTM_User, time=0.3, value=0.875}
         {arriveTangent=-0.0405, interpMode=RCIM_Cubic, leaveTangent=-0.0405, tangentMode=RCTM_User, time=1, value=0}
        ]
       }}}
     Scale Curve=&FN_FrontEndFloatModifier(Gameplay=0.7, Front End=0.35)
    )
    ScaleA=true
    ScaleRGB=true
   )
  ]
  rendererClasses [&NiagaraSpriteRendererProperties]
  renderers [{rendererClass=&NiagaraSpriteRendererProperties}]
 }
 NE_BodyGlow {
  emitterUpdateScript [
   /SetVariables(Emitter.OutlineThickness:NiagaraFloat=&FN_FrontEndFloatModifier(Gameplay=0.5, Front End=1)) {isSetParameters=true}
   &EmitterState()
   &SpawnBurst_Instantaneous(Spawn Count=8, Use Spawn Probability=false, UseLoopCountLimit=true)
  ]
  particleSpawnScript [
   &InitializeParticle(Color=<rgba 1 0 0 1>, Lifetime=4, Mesh Scale Mode=::NewEnumerator4, UsePositionOffset=true, Write Material Random=true)
   &PartitionParticles(Number Of Partitions=8)
  ]
  particleUpdateScript [
   &ParticleState(Kill Particles When Lifetime Has Elapsed=false)
   &ScaleMeshSize(Scale Factor=&VectorFromFloat(Value=0.8))
   &ScaleColor(
    Scale Alpha=&Lerp_Float(A=0, B=4, Alpha=&Subtract_Float(A=$User.ParaGlideLeanAlpha, B=0.5))
    Scale RGB=<1 0 0>
    ScaleA=true
    ScaleRGB=true
   ) {module=ScaleColor001}
   &SkeletalMeshLocation(
    Bone / Socket Index=&SelectIntFromArray(Direct Array Index=$Particles.PartitionParticles.Partition, Int Selection Array=DataInterface() {propertyValues {IntData [0,1,2,3,4,5,6,7]}})
    Bone Sampling Mode=::NewEnumerator4
    MaskSpawnGroup=false
    Orientation Sampling=::NewEnumerator1
    Skeletal Mesh=$System.Skeletal Mesh
   )
   &CameraOffset(Camera Offset Amount=20)
   &DynamicMaterialParameters(Module.Index0Float0Write=true, Module.Index0Float1Write=true, Module.Index0Float2Write=true, Module.Index0Float3Write=true)
  ]
  rendererClasses [&NiagaraMeshRendererProperties]
  renderers [
   {rendererClass=&NiagaraMeshRendererProperties, rendererIndex=0}
   {rendererClass=&NiagaraMeshRendererProperties, rendererIndex=1}
  ]
 }
 NE_BodyScan {
  emitterUpdateScript [
   &EmitterState(
    Inactive Response=::NewEnumerator1
    Life Cycle Mode=::NewEnumerator1
    Loop Behavior=::NewEnumerator1
    Loop Duration Mode=::NewEnumerator1
    UseLoopDelay=false
   )
   &SpawnBurst_Instantaneous(Use Spawn Probability=false, UseLoopCountLimit=true)
   /SetVariables(Emitter.V Panner Speed:NiagaraFloat=&FN_FrontEndFloatModifier(Gameplay=0.5, Front End=-0.5)) {isSetParameters=true}
  ]
  particleSpawnScript [
   &InitializeParticle(Color Mode=::NewEnumerator0, Sprite Size Mode=::NewEnumerator3, Uniform Sprite Size=150, UsePositionOffset=false, Write Material Random=true)
   &CameraOffset(Camera Offset Amount=5)
  ]
  particleUpdateScript [
   &ParticleState(Kill Particles When Lifetime Has Elapsed=false, Loop Particles Lifetime=true)
   &SkeletalMeshLocation(
    Bone / Socket Index=4
    Bone Sampling Mode=::NewEnumerator4
    MaskSpawnGroup=false
    Orientation Sampling=::NewEnumerator1
    Skeletal Mesh=$System.Skeletal Mesh
   )
   &ScaleColor(
    Color Scale:LinearColor=&LerpLinearColors(
     StartColor=<rgba 0 0 0 0>
     EndColor=<rgba 50 0 0.0008 1>
     LerpFactor=&Subtract_Float(A=$User.ParaGlideLeanAlpha, B=0.5)
    )
    Scale Mode=3
   )
   &ScaleColor(
    Scale Alpha=&FloatFromCurve(FloatCurve=DataInterface() {propertyValues {Curve {keys [
         {interpMode=RCIM_Cubic, tangentMode=RCTM_User, time=0, value=0}
         {interpMode=RCIM_Cubic, tangentMode=RCTM_User, time=0.5, value=1}
         {interpMode=RCIM_Cubic, tangentMode=RCTM_User, time=1, value=0}
        ]}}})
    Scale RGB=<10 0 0>
    ScaleA=true
    ScaleRGB=true
   ) {module=ScaleColor001}
  ]
  rendererClasses [&NiagaraSpriteRendererProperties]
  renderers [{rendererClass=&NiagaraSpriteRendererProperties}]
 }
 NE_Drone {
  emitterUpdateScript [
   &EmitterState()
   &SpawnBurst_Instantaneous(Spawn Count=1, Spawn Probability=HlslExpression() {hlslExpression="Emitter.LoopCount == 0"}, Use Spawn Probability=true)
   /SetVariables(Emitter.BladesSpeed:NiagaraFloat=&Lerp_Float(A=-3, B=-4, Alpha=$User.ParaGlideLeanAlpha)) {isSetParameters=true}
  ]
  particleSpawnScript [
   &InitializeParticle(Color=<rgba 5 0 0 1>, Lifetime=10, Mesh Scale Mode=::NewEnumerator4, UsePositionOffset=false, Write Material Random=true)
   /SetVariables() {isSetParameters=true}
  ]
  particleUpdateScript [
   &ParticleState(Kill Particles When Lifetime Has Elapsed=false, Loop Particles Lifetime=true)
   &SkeletalMeshLocation(
    Bone / Socket Index=4
    Bone Sampling Mode=::NewEnumerator4
    MaskSpawnGroup=false
    Orientation Sampling=::NewEnumerator1
    Sampled Position Offset=&Lerp_Vector(
     A=&MakeVector(
      X=&Waveform(
       "Waveform [1]"=::NewEnumerator12
       "[1] Input Position"=&RandomRangeFloat(Minimum=0, Maximum=1000)
       "[1] Phase Scale"=0.5
       "[1] Frequency"=1
       "[1] Amplitude Min/Max"=<100 -100>
      )
      Y=&Waveform(
       "Waveform [1]"=::NewEnumerator12
       "[1] Input Position"=&RandomRangeFloat(Minimum=1200, Maximum=1800)
       "[1] Phase Scale"=0.5
       "[1] Frequency"=1
       "[1] Amplitude Min/Max"=<-100 -100>
      )
      Z=&Waveform(
       "Waveform [1]"=::NewEnumerator12
       "[1] Input Position"=&RandomRangeFloat(Minimum=2200, Maximum=2800)
       "[1] Frequency"=1
       "[1] Amplitude Min/Max"=<-50 50>
      )
     )
     B=&Lerp_Vector(A=<0 -100 0>, B=<-50 -100 0>, Alpha=$User.ParaGlideLeanAlpha)
     Alpha=$User.ParaGlideLeanAlpha
    )
    Skeletal Mesh=$System.Skeletal Mesh
   ) {module=SkeletalMeshLocation001}
   /SetVariables(Particles.MeshOrientation=&MultiplyQuaternion(
     Quaternion A=$Particles.SkeletalMeshLocation001.SampledRotation
     Quaternion B=&LerpQuaternion(
      Quaternion A:Quat4f=/Game/Effects/Niagara/Modules/OrientMestToVectorFromCustomRot(
       InitialRotation:Quat4f=<1 1 -1 1>
       Look At Direction:Vector3f=<0 0 0>
       Reference Axis Vector:Vector3f=<0 0 0>
       RotationVector:Vector3f=/Niagara/DynamicInputs/Math/LerpMultipleVectors(
        Alpha:NiagaraFloat=&FloatFromCurve(FloatCurve=DataInterface() {propertyValues {Curve {keys [
             {interpMode=RCIM_Cubic, tangentMode=RCTM_User, time=0, value=1}
             {interpMode=RCIM_Cubic, tangentMode=RCTM_User, time=1, value=0}
             {interpMode=RCIM_Cubic, tangentMode=RCTM_User, time=2, value=1}
            ]}}})
        V 1:Vector3f=<-0.025 0.01 0.05>
        V 2:Vector3f=<0.025 -0.01 0>
        V 3:Vector3f=<-0.025 0.01 0.075>
        V 4:Vector3f=<0.025 -0.01 0>
        V 5:Vector3f=<-0.025 0.01 0.05>
        Vector Count:ENiagara_MultipleLerpCount=::NewEnumerator2
       )
      )
      Quaternion B:Quat4f=&LerpQuaternion(
       Quaternion A:Quat4f=<0.35 1 -0.35 1>
       Quaternion B:Quat4f=<0.175 1 -0.175 1>
       Lerp Factor:NiagaraFloat=&FloatFromCurve(FloatCurve=DataInterface() {propertyValues {Curve {
           defaultValue=0
           keys [
            {tangentMode=RCTM_User, time=0, value=0}
            <at 0.1 1>
            <at 0.2 0>
            <at 0.3 1>
            <at 0.4 0>
            <at 0.5 1>
            <at 0.6 0>
            <at 0.7 1>
            <at 0.8 0>
            <at 0.9 1>
            {tangentMode=RCTM_User, time=1, value=0}
           ]
          }}})
       Enable Spherical Linear Interpolation:NiagaraBool=true
      )
      Lerp Factor:NiagaraFloat=$User.ParaGlideLeanAlpha
      Enable Spherical Linear Interpolation:NiagaraBool=false
     )
    )) {isSetParameters=true}
   /SetVariables(Particles.Position=&Lerp_Vector(
     A=&Multiply_VectorByFloat(
      Vector=&MakeVector(
       X=&Waveform(
        "Waveform [1]"=::NewEnumerator12
        "[1] Input Position"=&RandomRangeFloat(Minimum=0, Maximum=1000)
        "[1] Phase Scale"=0.5
        "[1] Frequency"=1
        "[1] Amplitude Min/Max"=<150 -150>
       )
       Y=&Waveform(
        "Waveform [1]"=::NewEnumerator12
        "[1] Input Position"=&RandomRangeFloat(Minimum=1200, Maximum=1800)
        "[1] Phase Scale"=0.5
        "[1] Frequency"=1
        "[1] Amplitude Min/Max"=<100 -150>
       )
       Z=&Waveform(
        "Waveform [1]"=::NewEnumerator12
        "[1] Input Position"=&RandomRangeFloat(Minimum=2200, Maximum=2800)
        "[1] Frequency"=1
        "[1] Amplitude Min/Max"=<450 550>
       )
      )
      Float=0.35
     )
     B=Linked() {linkedVariable {name=StackContext.Position, type=NiagaraPosition}}
     Alpha=$User.ParaGlideLeanAlpha
    )) {isSetParameters=true}
   &SolveForcesAndVelocity(Clamp Velocity=false, Limit Acceleration=false)
   &ScaleColor(
    Scale Alpha=&Lerp_Float(A=0, B=1, Alpha=&Subtract_Float(A=$User.ParaGlideLeanAlpha, B=0.5))
    ScaleA=true
    ScaleRGB=true
   )
   &DynamicMaterialParameters(Module.Index0Float0Write=true, Module.Index0Float1Write=true, Module.Index0Float2Write=true, Module.Index0Float3Write=true)
  ]
  rendererClasses [&NiagaraMeshRendererProperties]
  renderers [
   {rendererClass=&NiagaraMeshRendererProperties, rendererIndex=0}
   {rendererClass=&NiagaraMeshRendererProperties, rendererIndex=1}
  ]
 }
 NE_HUD {
  emitterUpdateScript [
   &EmitterState(Life Cycle Mode=::NewEnumerator1, Loop Duration=0.5, UseLoopDelay=false)
   &SpawnBurst_Instantaneous(Spawn Probability=HlslExpression() {hlslExpression="Emitter.LoopCount == 0"}, Use Spawn Probability=true, UseLoopCountLimit=false)
   &SpawnBurst_Instantaneous(
    Spawn Group=1
    Spawn Probability=HlslExpression() {hlslExpression="Emitter.LoopCount == 0"}
    Use Spawn Probability=true
    UseLoopCountLimit=false
   ) {module=SpawnBurst_Instantaneous001}
   &SpawnBurst_Instantaneous(
    Spawn Group=2
    Spawn Probability=HlslExpression() {hlslExpression="Emitter.LoopCount == 0"}
    Use Spawn Probability=true
    UseLoopCountLimit=false
   ) {module=SpawnBurst_Instantaneous002}
   &SpawnBurst_Instantaneous(
    Spawn Group=3
    Spawn Probability=HlslExpression() {hlslExpression="Emitter.LoopCount == 0"}
    Use Spawn Probability=true
    UseLoopCountLimit=false
   ) {module=SpawnBurst_Instantaneous003}
   /SetVariables(
    Emitter.Scale U:NiagaraFloat=&Lerp_Float(
     A=&FloatFromCurve(FloatCurve=DataInterface() {propertyValues {Curve {keys [<at 0 0.75>,<at 0.5 1.5>,<at 1 0.75>]}}}, CurveIndex=$Emitter.NormalizedLoopAge)
     B=&FloatFromCurve(
      FloatCurve=DataInterface() {propertyValues {Curve {keys [<at 0 0.75>,<at 0.25 1.5>,<at 0.5 0.75>,<at 0.75 1.5>,<at 1 0.75>]}}}
      CurveIndex=$Emitter.NormalizedLoopAge
     )
     Alpha=$User.ParaGlideLeanAlpha
    )
    Emitter.Scale V:NiagaraFloat=&Lerp_Float(
     A=&FloatFromCurve(FloatCurve=DataInterface() {propertyValues {Curve {keys [<at 0 0.75>,<at 0.5 1.5>,<at 1 0.75>]}}}, CurveIndex=$Emitter.NormalizedLoopAge)
     B=&FloatFromCurve(
      FloatCurve=DataInterface() {propertyValues {Curve {keys [<at 0 0.75>,<at 0.25 1.5>,<at 0.5 0.75>,<at 0.75 1.5>,<at 1 0.75>]}}}
      CurveIndex=$Emitter.NormalizedLoopAge
     )
     Alpha=$User.ParaGlideLeanAlpha
    )
   ) {isSetParameters=true}
  ]
  particleSpawnScript [
   /SetVariables(Particles.VisibilityTag=&SpawnGroup()) {isSetParameters=true}
   &InitializeParticle(
    Color=&MergeHelper_ColorSwitch(
     0=<rgba 1 0 0 1>
     1=<rgba 1 0 0 1>
     2=&MakeCustomLinearColorFromBool(Bool=$User.bIsFrontEndPreview, True Color=<rgba 1.25 0 0 0>, False Color=<rgba 1.25 0 0 0.175>)
     3=&MakeCustomLinearColorFromBool(Bool=$User.bIsFrontEndPreview, True Color=<rgba 1.25 0 0 0>, False Color=<rgba 1.25 0 0 0.175>)
     Index=&SpawnGroup()
     Num=::NewEnumerator2
    )
    Lifetime=0.25
    Mesh Scale=<0.5 0.5 1.5>
    Mesh Scale Mode=::NewEnumerator4
    Position Offset=&SelectVectorFromArray(Direct Array Index=$Particles.VisibilityTag, Vector Selection Array=DataInterface() {propertyValues {FloatData [<-50 100 50>, <88 86 50>]}})
    Sprite Size=&MergeHelper_Vector2Switch(0=<0 0>, 1=<0 0>, 2=<125 100>, 3=<200 100>, 4=?, Index=&SpawnGroup(), Num=::NewEnumerator2)
    Sprite Size Mode=::NewEnumerator4
    UsePositionOffset=true
    Write Material Random=true
   )
   &InitialMeshOrientation(
    Mesh Axis To Orient=<0 0 -1>
    Mesh Orientation Mode=::NewEnumerator2
    Orientation Vector=<0 1 0>
    Rotation Coordinate Space=::NewEnumerator2
    Use Rotation Vector=true
   )
  ]
  particleUpdateScript [
   &ParticleState(
    Kill Particles When Lifetime Has Elapsed=false
    Let Infinitely Lived Particles Die When Emitter Deactivates=true
    Loop Particles Lifetime=true
   )
   &SolveForcesAndVelocity(Clamp Velocity=false, Limit Acceleration=false)
   &SkeletalMeshLocation(
    Bone / Socket Index=4
    Bone Sampling Mode=::NewEnumerator4
    MaskSpawnGroup=false
    Sampled Position Offset=&SelectVectorFromArray(
     Direct Array Index=$Particles.VisibilityTag
     Vector Selection Array=DataInterface() {propertyValues {FloatData [<80 20 50>, <100 0 -50>, <80 -5 50>, <100 -25 -50>]}}
    )
    Skeletal Mesh=$System.Skeletal Mesh
   )
   /SetVariables(Particles.MeshOrientation=&MultiplyQuaternion(
     Quaternion A=$Particles.SkeletalMeshLocation.SampledRotation
     Quaternion B=/Game/Effects/Niagara/Modules/MergeEmitters/MergeHelper_QuatSwitch(
      0:Quat4f=<1 -0.3 0.2 1>
      1:Quat4f=<1 0.2 -0.3 1>
      2:Quat4f=?
      3:Quat4f=?
      4:Quat4f=?
      5:Quat4f=?
      6:Quat4f=?
      7:Quat4f=?
      8:Quat4f=?
      9:Quat4f=?
      Index:NiagaraInt32=$Particles.VisibilityTag
      Num:MergeHelper_SwitcherEnum=::NewEnumerator0
     )
    )) {isSetParameters=true}
   &DynamicMaterialParameters(
    Index 0 Param 1=&Lerp_Float(A=3, B=10, Alpha=$User.ParaGlideLeanAlpha)
    Index 0 Param 2=&MakeCustomFloatFromBool(Bool=$User.bIsFrontEndPreview, True Float=5, False Float=0)
    Module.Index0Float0Write=true
    Module.Index0Float1Write=true
    Module.Index0Float2Write=true
    Module.Index0Float3Write=true
   )
   &ScaleColor(
    Scale Alpha=&MergeHelper_FloatSwitch(0=1, 1=1, 2=1, 3=1, Index=&SpawnGroup(), Num=::NewEnumerator2)
    Scale RGB=&MergeHelper_VectorSwitch(
     0=&VectorFromFloat(Value=&FNTimeOfDayScalar(Night Scale=1, Morning Scale=1.25, Day Scale=1.5, Evening Scale=1.25))
     1=&VectorFromFloat(Value=&FNTimeOfDayScalar(Night Scale=1, Morning Scale=1.25, Day Scale=1.5, Evening Scale=1.25))
     2=<1 1 1>
     3=<1 1 1>
     Index=&SpawnGroup()
     Num=::NewEnumerator2
    )
    ScaleA=true
    ScaleRGB=true
   )
  ]
  rendererClasses [&NiagaraMeshRendererProperties, &NiagaraSpriteRendererProperties]
  renderers [
   {rendererClass=&NiagaraMeshRendererProperties, rendererIndex=0}
   {rendererClass=&NiagaraMeshRendererProperties, rendererIndex=1}
   {rendererClass=&NiagaraMeshRendererProperties, rendererIndex=2}
   {rendererClass=&NiagaraSpriteRendererProperties, rendererIndex=3}
   {rendererClass=&NiagaraSpriteRendererProperties, rendererIndex=4}
  ]
 }
}
summary {
 systemName=NS_Contrail_BrokenGravel
 emitters [
  {emitterName=NE_Drone, rendererClasses [&NiagaraMeshRendererProperties]}
  {emitterName=NE_HUD, rendererClasses [&NiagaraMeshRendererProperties, &NiagaraSpriteRendererProperties]}
  {emitterName=NE_BodyGlow, rendererClasses [&NiagaraMeshRendererProperties]}
  {emitterName=FN_SpeedLines, simTarget=GPUComputeSim, rendererClasses [&NiagaraSpriteRendererProperties]}
  {emitterName=NE_BodyScan, rendererClasses [&NiagaraSpriteRendererProperties]}
 ]
 userVariables [
  User.bIsFrontEnd {type=NiagaraBool}
  User.bIsFrontEndPreview {type=NiagaraBool}
  User.bisHeadHidden {type=NiagaraBool}
  User.IsLargeBody {type=NiagaraBool}
  User.ParaGlideLeanAlpha {type=NiagaraFloat, defaultValue=0, description=""}
  User.RibbonIDOffset {type=NiagaraInt32}
  User.SphereRadius {type=NiagaraFloat, defaultValue=1500, description=""}
  User.Team {type=NiagaraInt32}
  User.TW_FXReactvity {type=NiagaraFloat, defaultValue=0.5, description=""}
  User.VariantID {type=NiagaraFloat, defaultValue=0, description=""}
 ]
}
